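# Post-process Animation Blueprint for the CarolinaChili character head that dampens head rotation by counter-rotating the head bone.
# Each update reads the pawn mesh's head delta from reference pose relative to neck_02, then negates and scales each axis by its percentage variable.
# A Modify Bone node adds this rotation in bone space to the input pose, blending out during emotes or in the front end.
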AnimBlueprint /BRCosmetics/Characters/Player/Male/Medium/Heads/M_MED_CarolinaChili_Head/Meshes/M_MED_CarolinaChili_Head_PPAnimBP
target_skeleton=/Game/Characters/Player/Common/Fortnite_Base_Head/Fortnite_Base_Head_Export_Skeleton
graphs [
 AnimGraph {nodes [
   n1 = "Transform (Modify) Bone - Bone: head"!(
    ComponentPose=Local To Component(LocalPose=Input Pose())
    bAlphaBoolEnabled:bool=Get EmoteNotPlaying() -> (EmoteNotPlaying:bool)
    Translation="0,0,0"
    Rotation=Get head rot() -> (head rot:Rotator)
    Scale="1,1,1"
   ) -> (Pose) {
    alphaInputType=Bool
    rotationMode=BMM_Additive
    rotationSpace=BCS_BoneSpace
    scaleMode=BMM_Additive
    scaleSpace=BCS_BoneSpace
    alphaBoolBlend {blendInTime=1, blendOption=Linear, blendOutTime=1}
    boneToModify {boneName=head}
   }
   Output Pose(Result=Component To Local(ComponentPose=@n1))
  ]}
 EventGraph {nodes [
   Event BlueprintInitializeAnimation()
   > n1 = Cast To PlayerPawn_Athena(Object=TryGetPawnOwner())
   > n2 = Set Pawn Owner(Pawn Owner:Player Pawn Athena Object Reference=@n1) -> (Output_Get:Player Pawn Athena Object Reference)
   > Set Head_Mesh(Head_Mesh:Skeletal Mesh Component Object Reference=GetSkeletalMeshForPartType(self=@n2, PartType=Head)) -> (Output_Get:Skeletal Mesh Component Object Reference)
   Event BlueprintUpdateAnimation()
   n3 = Get Mesh(self=Get Pawn Owner() -> (Pawn Owner:Player Pawn Athena Object Reference))
   n4 = BreakRotator(InRot=BreakTransform(InTransform=GetDeltaTransformFromRefPose(self=@n3, BoneName=head, BaseName=neck_02)).Rotation)
   n5 = Get Y percentage() -> (Y percentage:float)
   n6 = Get X percentage() -> (X percentage:float)
   n7 = Get Z percentage() -> (Z percentage:float)
   > Set head rot(head rot:Rotator=MakeRotator(
     Roll=float * float!(A=@n4, B=-1, C:double=@n6) -> (ReturnValue)
     Pitch=float * float(A=@n4.Pitch, B=-1, C=@n5)
     Yaw=float * float(A=@n4.Yaw, B=-1, C=@n7)
    )) -> (Output_Get:Rotator)
   n8 = Get bIsInFrontEnd()
   n9 = Get bIsPlayingEmote()
   > Set EmoteNotPlaying(EmoteNotPlaying:bool=AND Boolean(A=NOT Boolean(A=@n9), B=NOT Boolean(A=@n8))) -> (Output_Get:bool)
  ]}
]
variables [
 EmoteNotPlaying {type=boolean}
 frontEndNotPlaying {type=boolean}
 head rot {type=Rotator}
 Head_Mesh {type=SkeletalMeshComponent}
 Pawn Owner {type=PlayerPawn_Athena_C}
 X percentage {type=number, default=0.7}
 Y percentage {type=number, default=0.7}
 Z percentage {type=number, default=0.7}
]
